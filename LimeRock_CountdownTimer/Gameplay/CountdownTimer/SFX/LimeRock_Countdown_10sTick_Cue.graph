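# Sound cue for the 10-second tick of the LimeRock countdown timer.
# A single wave player plays the tick sound straight into the output, with the cue volume reduced.
# It is spatialized through the countdown attenuation asset, routed to the Fortnite events sound class, and uses a concurrency override that caps simultaneous instances at two.
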SoundCue /LimeRock_CountdownTimer/Gameplay/CountdownTimer/SFX/LimeRock_Countdown_10sTick_Cue
root=WavePlayer_0
nodes [{id=WavePlayer_0, properties {soundWaveAssetPtr=&/LimeRock_Countdown_10sTick}}]
properties {
 attenuationSettings=&/LimeRock_Countdown_Att
 bOverrideConcurrency=true
 maxDistance=4200
 soundClassObject=&Fort_Events
 volumeMultiplier=0.45
 concurrencyOverrides {maxCount=2}
}
