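# Sound cue for the 60-second tick of the LimeRock countdown timer.
# A single wave player feeds the LimeRock_Countdown_60sTick wave straight to the output.
# It plays at reduced volume in the Fort_Events sound class, using a dedicated 60s attenuation asset.
# A concurrency override caps how many instances play at once.
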SoundCue /LimeRock_CountdownTimer/Gameplay/CountdownTimer/SFX/LimeRock_Countdown_60sTick_Cue
root=WavePlayer_0
nodes [{id=WavePlayer_0, properties {soundWaveAssetPtr=&/LimeRock_Countdown_60sTick}}]
properties {
 attenuationSettings=&/LimeRock_Countdown_60s_Att
 bOverrideConcurrency=true
 maxDistance=11500
 soundClassObject=&Fort_Events
 volumeMultiplier=0.2
 concurrencyOverrides {maxCount=2}
}
