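# Sound cue for the normal tick of the LimeRock countdown timer.
# A single wave player feeds the tick sound wave straight to the output at reduced volume, with no randomization or modulation.
# It routes through the Fort_Events sound class, uses a dedicated attenuation asset, and overrides concurrency to limit simultaneous instances.
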SoundCue /LimeRock_CountdownTimer/Gameplay/CountdownTimer/SFX/LimeRock_Countdown_NormalTick_Cue
root=WavePlayer_1
nodes [{id=WavePlayer_1, properties {soundWaveAssetPtr=&/LimeRock_Countdown_NormalTick}}]
properties {
 attenuationSettings=&/ATT_LimeRock_Countdown_NormalTick
 bOverrideConcurrency=true
 maxDistance=1000
 soundClassObject=&Fort_Events
 volumeMultiplier=0.3
 concurrencyOverrides {maxCount=2}
}
