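# Sound cue for the shutdown sound of the LimeRock countdown timer.
# A single wave player feeds the shutdown wave straight to the output at reduced volume, spatialized through the countdown's attenuation settings.
# It routes to an important in-game mission sound class, caps simultaneous instances with a concurrency override, and sends a light signal to the distant reverb submix.
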SoundCue /LimeRock_CountdownTimer/Gameplay/CountdownTimer/SFX/LimeRock_Countdown_Shutdown_Cue
root=WavePlayer_1
nodes [{id=WavePlayer_1, properties {soundWaveAssetPtr=&/LimeRock_Countdown_Shutdown}}]
properties {
 attenuationSettings=&/LimeRock_Countdown_Att
 bOverrideConcurrency=true
 maxDistance=4200
 soundClassObject=&Fort_Mission_Ingame_Important_PSM
 volumeMultiplier=0.2
 concurrencyOverrides {maxCount=2}
 soundSubmixSends [{minSendLevel=0, sendLevel=0.2, soundSubmix=&Fortnite_Reverb_Distant}]
}
