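# One-shot stereo MetaSound source that plays the LimeRock countdown stinger and exposes a float TickOut output.
# The Wave Player starts on play, feeds both channels and signals the source finished when the wave ends.
# Alongside it, a one-second TriggerRepeat drives the Set input of a Value node.
# That node's Init Value comes from a deferred float variable minus one, and its output feeds both the variable and TickOut.
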MetaSoundSource /LimeRock_CountdownTimer/Gameplay/CountdownTimer/SFX/MSS_LimeRock_Countdown_STNG
inputs [
 UE.Source.OnPlay {type=Trigger}
 Wave Asset {type=WaveAsset}
]
interfaces [UE.OutputFormat.Stereo:1.0, UE.Source.OneShot:1.0, UE.Source:1.1]
outputs [
 TickOut {type=Float}
 "UE.OutputFormat.Stereo.Audio:0" {type=Audio}
 "UE.OutputFormat.Stereo.Audio:1" {type=Audio}
 UE.Source.OneShot.OnFinished {type=Trigger}
]
pages [Default {
  input_defaults {UE.Source.OnPlay=null, Wave Asset=&/MX_S37_LimeRock_Countdown_STNG}
  nodes [
   n1 = VariableDeferredAccessor.Float(Variable=InitVariable.Float().Variable)
   ValueValue_0 = Value.Value:Float(
    Init Value=UE.Subtract:Float(PrimaryOperand=@n1.Value) {also {defaults {AdditionalOperands=1}}}.Out
    Set=UE.TriggerRepeat(Start=^UE.Source.OnPlay) {also {defaults {Period=1}}}.RepeatOut
   )
   VariableMutator.Float(Value=@ValueValue_0.Output Value, Variable=@n1.Variable)
   WavePlayer_0 = UE.Wave Player:Stereo(Play=^UE.Source.OnPlay, Wave Asset=^Wave Asset) {also {version [1,1]}}
  ]
  links [
   {dst="@TickOut", src=ValueValue_0.Output Value}
   {dst="@UE.Source.OneShot.OnFinished", src=WavePlayer_0.On Finished}
   {dst="@UE.OutputFormat.Stereo.Audio:0", src=WavePlayer_0.Out Left}
   {dst="@UE.OutputFormat.Stereo.Audio:1", src=WavePlayer_0.Out Right}
  ]
 }]
properties {
 assetImportData=&self:AssetImportData
 numChannels=2
 outputFormat=Stereo
 sampleRate=48000
}
