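# Sound cue that plays the music stinger for the LimeRock countdown timer.
# A single wave player feeds the LimeRock countdown stinger wave straight to the output.
# It routes through the Fortnite music sound class, boosts volume slightly, and uses a large point-of-interest attenuation asset so the stinger is spatialized over a long range.
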SoundCue /LimeRock_CountdownTimer/Gameplay/CountdownTimer/SFX/SC_LimeRock_Countdown_STNG
root=WavePlayer_1
nodes [{id=WavePlayer_1, properties {soundWaveAssetPtr=&/MX_S37_LimeRock_Countdown_STNG}}]
properties {
 attenuationSettings=&/ATT_LimeRock_POI_Large
 maxDistance=40000
 soundClassObject=/Game/Sounds/FortSoundClassesAndModes/Fort_Music
 volumeMultiplier=1.3
}
