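# Looping sound cue for the LimeRock countdown timer's platform, routed to the Fort_Events sound class.
# A looping wave player feeds an Enveloper whose volume curve starts quiet, stays low, then rises sharply to about full level within the first second.
# A dedicated attenuation asset handles spatial falloff, and the cue's overall volume is reduced.
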SoundCue /LimeRock_CountdownTimer/Gameplay/CountdownTimer/SFX/SC_LimeRock_Seq_Timer_Platform_Loop
root=Enveloper_0
nodes [
 {
  id=Enveloper_0
  children [WavePlayer_0]
  properties {
   pitchCurve {editorCurveData {keys [{time=0}]}}
   volumeCurve {editorCurveData {keys [<at 0 0.1>,<at 0.4928 0.1392>,<at 0.6128 0.7938>,<at 0.8 1.0119>]}}
  }
 }
 {
  id=WavePlayer_0
  properties {bLooping=true, soundWaveAssetPtr=&/LimeRock_TimerPlatform_Loop}
 }
]
properties {
 attenuationSettings=&/ATT_LimeRock_Timer_Platform_Loop
 maxDistance=3500
 soundClassObject=&Fort_Events
 volumeMultiplier=0.4
}
